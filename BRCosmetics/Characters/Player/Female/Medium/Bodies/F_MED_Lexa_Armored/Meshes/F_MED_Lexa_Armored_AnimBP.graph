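# Anim Blueprint for the armored Lexa cosmetic body that copies the attached parent's pose and adds secondary motion to armor and choker bones.
# Shoulder pads shift additively while bIsTargeting is true.
# Arm and leg armor get spring-based AnimDynamics, then a floating bob from two sine-wave vectors computed each update from real time.
# The choker base is constrained between clavicles and upper spine, then its parent bone is simulated with angular limits.
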AnimBlueprint /BRCosmetics/Characters/Player/Female/Medium/Bodies/F_MED_Lexa_Armored/Meshes/F_MED_Lexa_Armored_AnimBP
target_skeleton=&/F_MED_Lexa_Armored_Skeleton
graphs [
 AnimGraph {nodes [
   n1 = Get Shoulderpads_ADS_Alpha() -> (Shoulderpads_ADS_Alpha:double)
   n2 = "Transform (Modify) Bone - Bone: dyn_shoulderpad_l_parent"(
    ComponentPose=Local To Component(LocalPose=Copy Pose From Mesh() {bUseAttachedParent=true})
    Alpha=@n1
    Translation="3,0,0"
    Rotation="0,0,0"
    Scale="1,1,1"
   ) {
    translationMode=BMM_Additive
    translationSpace=BCS_BoneSpace
    alphaBoolBlend {blendOption=Linear}
    boneToModify {boneName=dyn_shoulderpad_l_parent}
    translation {x=3}
   }
   n3 = "Transform (Modify) Bone - Bone: dyn_shoulderpad_r_parent"(ComponentPose=@n2, Alpha=@n1, Translation="3,0,0", Rotation="0,0,0", Scale="1,1,1") {
    translationMode=BMM_Additive
    translationSpace=BCS_BoneSpace
    alphaBoolBlend {blendOption=Linear}
    boneToModify {boneName=dyn_shoulderpad_r_parent}
    translation {x=3}
   }
   n4 = Anim Dynamics - Bone: dyn_armor_upper_arm_r_parent(ComponentPose=@n3, Alpha=0.3, ExternalForce="0,0,0") {
    angularDampingOverride=0.9
    angularSpringConstant=150
    bAngularSpring=true
    bGravityOverrideInSimSpace=true
    bLinearSpring=true
    bOverrideAngularDamping=true
    bOverrideLinearDamping=true
    gravityScale=0
    linearDampingOverride=0.995
    linearSpringConstant=50
    simulationSpace=World
    alphaBoolBlend {blendOption=Linear}
    alphaScaleBiasClamp {bInterpResult=true, interpSpeedDecreasing=4, interpSpeedIncreasing=6}
    boundBone {boneName=dyn_armor_upper_arm_r_parent}
    componentLinearAccScale=<0.05 0.05 0.05>
    componentLinearVelScale=<0.8 0.8 0.6>
    physicsBodyDefinitions [{
      sphereCollisionRadius=0
      boundBone {boneName=dyn_armor_upper_arm_r_parent}
      boxExtents=<50 50 50>
      constraintSetup {angularLimitsMax=<0 0 0>, angularLimitsMin=<0 0 0>, angularTarget=<0 0 0>, linearAxesMax=<4 1 1>, linearAxesMin=<-4 -1 -1>}
      localJointOffset=<0 -3 0>
     }]
    relativeSpaceBone {boneName=clavicle_r}
   }
   n5 = Anim Dynamics - Bone: dyn_armor_lower_arm_r_parent(ComponentPose=@n4, Alpha=0.17, ExternalForce="0,0,0") {
    angularDampingOverride=0.9
    angularSpringConstant=150
    bAngularSpring=true
    bGravityOverrideInSimSpace=true
    bLinearSpring=true
    bOverrideAngularDamping=true
    bOverrideLinearDamping=true
    gravityScale=0
    linearDampingOverride=0.995
    linearSpringConstant=50
    simulationSpace=World
    alphaBoolBlend {blendOption=Linear}
    alphaScaleBiasClamp {bInterpResult=true, interpSpeedDecreasing=4, interpSpeedIncreasing=6}
    boundBone {boneName=dyn_armor_lower_arm_r_parent}
    componentLinearAccScale=<0.1 0.1 0.3>
    componentLinearVelScale=<0.8 0.8 0.6>
    physicsBodyDefinitions [{
      sphereCollisionRadius=0
      boundBone {boneName=dyn_armor_lower_arm_r_parent}
      boxExtents=<50 50 50>
      constraintSetup {angularLimitsMax=<0 0 0>, angularLimitsMin=<0 0 0>, angularTarget=<0 0 0>, linearAxesMax=<4 1 1>, linearAxesMin=<-4 -1 -1>}
      localJointOffset=<0 -3 0>
     }]
    relativeSpaceBone {boneName=clavicle_r}
   }
   n6 = Anim Dynamics - Bone: dyn_armor_upper_leg_r_parent(ComponentPose=@n5, Alpha=0.3, ExternalForce="0,0,0") {
    angularDampingOverride=0.9
    angularSpringConstant=150
    bAngularSpring=true
    bGravityOverrideInSimSpace=true
    bLinearSpring=true
    bOverrideAngularDamping=true
    bOverrideLinearDamping=true
    gravityScale=0
    linearDampingOverride=0.995
    linearSpringConstant=50
    simulationSpace=World
    alphaBoolBlend {blendOption=Linear}
    alphaScaleBiasClamp {bInterpResult=true, interpSpeedDecreasing=4, interpSpeedIncreasing=6}
    boundBone {boneName=dyn_armor_upper_leg_r_parent}
    componentLinearAccScale=<0.05 0.05 0.05>
    componentLinearVelScale=<0.8 1 0.6>
    physicsBodyDefinitions [{
      sphereCollisionRadius=0
      boundBone {boneName=dyn_armor_upper_leg_r_parent}
      boxExtents=<50 50 50>
      constraintSetup {angularLimitsMax=<0 0 0>, angularLimitsMin=<0 0 0>, angularTarget=<0 0 0>, linearAxesMax=<4 1 1>, linearAxesMin=<-4 -1 -1>}
      localJointOffset=<0 -3 0>
     }]
    relativeSpaceBone {boneName=root}
   }
   n7 = Anim Dynamics - Bone: dyn_armor_lower_leg_r_parent(ComponentPose=@n6, Alpha=0.17, ExternalForce="0,0,0") {
    angularDampingOverride=0.9
    angularSpringConstant=150
    bAngularSpring=true
    bGravityOverrideInSimSpace=true
    bLinearSpring=true
    bOverrideAngularDamping=true
    bOverrideLinearDamping=true
    gravityScale=0
    linearDampingOverride=0.995
    linearSpringConstant=50
    simulationSpace=World
    alphaBoolBlend {blendOption=Linear}
    alphaScaleBiasClamp {bInterpResult=true, interpSpeedDecreasing=4, interpSpeedIncreasing=6}
    boundBone {boneName=dyn_armor_lower_leg_r_parent}
    componentLinearAccScale=<0.05 0.05 0.05>
    componentLinearVelScale=<0.8 1 0.6>
    physicsBodyDefinitions [{
      sphereCollisionRadius=0
      boundBone {boneName=dyn_armor_lower_leg_r_parent}
      boxExtents=<50 50 50>
      constraintSetup {angularLimitsMax=<0 0 0>, angularLimitsMin=<0 0 0>, angularTarget=<0 0 0>, linearAxesMax=<4 1 1>, linearAxesMin=<-4 -1 -1>}
      localJointOffset=<0 -3 0>
     }]
    relativeSpaceBone {boneName=root}
   }
   n8 = Anim Dynamics - Bone: dyn_armor_upper_arm_l_parent(ComponentPose=@n7, Alpha=0.3, ExternalForce="0,0,0") {
    angularDampingOverride=0.9
    angularSpringConstant=150
    bAngularSpring=true
    bGravityOverrideInSimSpace=true
    bLinearSpring=true
    bOverrideAngularDamping=true
    bOverrideLinearDamping=true
    gravityScale=0
    linearDampingOverride=0.995
    linearSpringConstant=50
    simulationSpace=World
    alphaBoolBlend {blendOption=Linear}
    alphaScaleBiasClamp {bInterpResult=true, interpSpeedDecreasing=4, interpSpeedIncreasing=6}
    boundBone {boneName=dyn_armor_upper_arm_l_parent}
    componentLinearAccScale=<0.1 0.1 0.3>
    componentLinearVelScale=<0.8 0.8 0.6>
    physicsBodyDefinitions [{
      sphereCollisionRadius=0
      boundBone {boneName=dyn_armor_upper_arm_l_parent}
      boxExtents=<50 50 50>
      constraintSetup {angularLimitsMax=<0 0 0>, angularLimitsMin=<0 0 0>, angularTarget=<0 0 0>, linearAxesMax=<4 1 1>, linearAxesMin=<-4 -1 -1>}
      localJointOffset=<0 -3 0>
     }]
    relativeSpaceBone {boneName=clavicle_l}
   }
   n9 = Anim Dynamics - Bone: dyn_armor_lower_arm_l_parent(ComponentPose=@n8, Alpha=0.17, ExternalForce="0,0,0") {
    angularDampingOverride=0.9
    angularSpringConstant=150
    bAngularSpring=true
    bGravityOverrideInSimSpace=true
    bLinearSpring=true
    bOverrideAngularDamping=true
    bOverrideLinearDamping=true
    gravityScale=0
    linearDampingOverride=0.995
    linearSpringConstant=50
    simulationSpace=World
    alphaBoolBlend {blendOption=Linear}
    alphaScaleBiasClamp {bInterpResult=true, interpSpeedDecreasing=4, interpSpeedIncreasing=6}
    boundBone {boneName=dyn_armor_lower_arm_l_parent}
    componentLinearAccScale=<0.05 0.05 0.05>
    componentLinearVelScale=<0.8 0.8 0.6>
    physicsBodyDefinitions [{
      sphereCollisionRadius=0
      boundBone {boneName=dyn_armor_lower_arm_l_parent}
      boxExtents=<50 50 50>
      constraintSetup {angularLimitsMax=<0 0 0>, angularLimitsMin=<0 0 0>, angularTarget=<0 0 0>, linearAxesMax=<4 1 1>, linearAxesMin=<-4 -1 -1>}
      localJointOffset=<0 -3 0>
     }]
    relativeSpaceBone {boneName=clavicle_l}
   }
   n10 = Anim Dynamics - Bone: dyn_armor_upper_leg_l_parent(ComponentPose=@n9, Alpha=0.3, ExternalForce="0,0,0") {
    angularDampingOverride=0.9
    angularSpringConstant=150
    bAngularSpring=true
    bGravityOverrideInSimSpace=true
    bLinearSpring=true
    bOverrideAngularDamping=true
    bOverrideLinearDamping=true
    gravityScale=0
    linearDampingOverride=0.995
    linearSpringConstant=50
    simulationSpace=World
    alphaBoolBlend {blendOption=Linear}
    alphaScaleBiasClamp {bInterpResult=true, interpSpeedDecreasing=4, interpSpeedIncreasing=6}
    boundBone {boneName=dyn_armor_upper_leg_l_parent}
    componentLinearAccScale=<0.1 0.1 0.3>
    componentLinearVelScale=<0.8 1 0.6>
    physicsBodyDefinitions [{
      sphereCollisionRadius=0
      boundBone {boneName=dyn_armor_upper_leg_l_parent}
      boxExtents=<50 50 50>
      constraintSetup {angularLimitsMax=<0 0 0>, angularLimitsMin=<0 0 0>, angularTarget=<0 0 0>, linearAxesMax=<4 1 1>, linearAxesMin=<-4 -1 -1>}
      localJointOffset=<0 -3 0>
     }]
    relativeSpaceBone {boneName=root}
   }
   n11 = Anim Dynamics - Bone: dyn_armor_lower_leg_l_parent(ComponentPose=@n10, Alpha=0.17, ExternalForce="0,0,0") {
    angularDampingOverride=0.9
    angularSpringConstant=150
    bAngularSpring=true
    bGravityOverrideInSimSpace=true
    bLinearSpring=true
    bOverrideAngularDamping=true
    bOverrideLinearDamping=true
    gravityScale=0
    linearDampingOverride=0.995
    linearSpringConstant=50
    simulationSpace=World
    alphaBoolBlend {blendOption=Linear}
    alphaScaleBiasClamp {bInterpResult=true, interpSpeedDecreasing=4, interpSpeedIncreasing=6}
    boundBone {boneName=dyn_armor_lower_leg_l_parent}
    componentLinearAccScale=<0.1 0.1 0.3>
    componentLinearVelScale=<0.8 1 0.6>
    physicsBodyDefinitions [{
      sphereCollisionRadius=0
      boundBone {boneName=dyn_armor_lower_leg_l_parent}
      boxExtents=<50 50 50>
      constraintSetup {angularLimitsMax=<0 0 0>, angularLimitsMin=<0 0 0>, angularTarget=<0 0 0>, linearAxesMax=<4 1 1>, linearAxesMin=<-4 -1 -1>}
      localJointOffset=<0 -3 0>
     }]
    relativeSpaceBone {boneName=root}
   }
   n12 = Get Spike_Float_A() -> (Spike_Float_A:Vector)
   n13 = "Transform (Modify) Bone - Bone: dyn_armor_upper_leg_r_parent"(ComponentPose=@n11, Alpha=1, Translation=@n12, Rotation="0,0,0", Scale="1,1,1") {
    translationMode=BMM_Additive
    translationSpace=BCS_BoneSpace
    alphaBoolBlend {blendOption=Linear}
    alphaScaleBiasClamp {bInterpResult=true, interpSpeedDecreasing=4, interpSpeedIncreasing=6}
    boneToModify {boneName=dyn_armor_upper_leg_r_parent}
   }
   n14 = Get Spike_Float_B() -> (Spike_Float_B:Vector)
   n15 = "Transform (Modify) Bone - Bone: dyn_armor_lower_leg_r_parent"(ComponentPose=@n13, Alpha=1, Translation=@n14, Rotation="0,0,0", Scale="1,1,1") {
    translationMode=BMM_Additive
    translationSpace=BCS_BoneSpace
    alphaBoolBlend {blendOption=Linear}
    alphaScaleBiasClamp {bInterpResult=true, interpSpeedDecreasing=4, interpSpeedIncreasing=6}
    boneToModify {boneName=dyn_armor_lower_leg_r_parent}
   }
   n16 = "Transform (Modify) Bone - Bone: dyn_armor_upper_leg_l_parent"(ComponentPose=@n15, Alpha=1, Translation=@n12, Rotation="0,0,0", Scale="1,1,1") {
    translationMode=BMM_Additive
    translationSpace=BCS_BoneSpace
    alphaBoolBlend {blendOption=Linear}
    alphaScaleBiasClamp {bInterpResult=true, interpSpeedDecreasing=4, interpSpeedIncreasing=6}
    boneToModify {boneName=dyn_armor_upper_leg_l_parent}
   }
   n17 = "Transform (Modify) Bone - Bone: dyn_armor_lower_leg_l_parent"(ComponentPose=@n16, Alpha=1, Translation=@n14, Rotation="0,0,0", Scale="1,1,1") {
    translationMode=BMM_Additive
    translationSpace=BCS_BoneSpace
    alphaBoolBlend {blendOption=Linear}
    alphaScaleBiasClamp {bInterpResult=true, interpSpeedDecreasing=4, interpSpeedIncreasing=6}
    boneToModify {boneName=dyn_armor_lower_leg_l_parent}
   }
   n18 = "Transform (Modify) Bone - Bone: dyn_armor_upper_arm_r_parent"(ComponentPose=@n17, Alpha=1, Translation=@n12, Rotation="0,0,0", Scale="1,1,1") {
    translationMode=BMM_Additive
    translationSpace=BCS_BoneSpace
    alphaBoolBlend {blendOption=Linear}
    alphaScaleBiasClamp {bInterpResult=true, interpSpeedDecreasing=4, interpSpeedIncreasing=6}
    boneToModify {boneName=dyn_armor_upper_arm_r_parent}
   }
   n19 = "Transform (Modify) Bone - Bone: dyn_armor_lower_arm_r_parent"(ComponentPose=@n18, Alpha=1, Translation=@n14, Rotation="0,0,0", Scale="1,1,1") {
    translationMode=BMM_Additive
    translationSpace=BCS_BoneSpace
    alphaBoolBlend {blendOption=Linear}
    alphaScaleBiasClamp {bInterpResult=true, interpSpeedDecreasing=4, interpSpeedIncreasing=6}
    boneToModify {boneName=dyn_armor_lower_arm_r_parent}
   }
   n20 = "Transform (Modify) Bone - Bone: dyn_armor_upper_arm_l_parent"(ComponentPose=@n19, Alpha=1, Translation=@n12, Rotation="0,0,0", Scale="1,1,1") {
    translationMode=BMM_Additive
    translationSpace=BCS_BoneSpace
    alphaBoolBlend {blendOption=Linear}
    alphaScaleBiasClamp {bInterpResult=true, interpSpeedDecreasing=4, interpSpeedIncreasing=6}
    boneToModify {boneName=dyn_armor_upper_arm_l_parent}
   }
   n21 = "Transform (Modify) Bone - Bone: dyn_armor_lower_arm_l_parent"(ComponentPose=@n20, Alpha=1, Translation=@n14, Rotation="0,0,0", Scale="1,1,1") {
    translationMode=BMM_Additive
    translationSpace=BCS_BoneSpace
    alphaBoolBlend {blendOption=Linear}
    alphaScaleBiasClamp {bInterpResult=true, interpSpeedDecreasing=4, interpSpeedIncreasing=6}
    boneToModify {boneName=dyn_armor_lower_arm_l_parent}
   }
   n22 = "Constraint - dyn_choker_base (PPP)"!(ComponentPose=@n21, Alpha=1, ConstraintWeights_0=0.0349, ConstraintWeights_1:float=0.9337, ConstraintWeights_2:float=0.0314) -> (Pose) {
    alphaBoolBlend {blendOption=Linear}
    boneToModify {boneName=dyn_choker_base}
    constraintSetup [{targetBone {boneName=clavicle_r}}, {targetBone {boneName=spine_05}}, {targetBone {boneName=clavicle_l}}]
    constraintWeights [1,1,1]
   }
   n23 = Anim Dynamics - Bone: dyn_choker_parent(ComponentPose=@n22, Alpha=1, ExternalForce="0,0,0") {
    angularDampingOverride=0.999
    bGravityOverrideInSimSpace=true
    bOverrideAngularDamping=true
    bOverrideLinearDamping=true
    linearDampingOverride=0.999
    simulationSpace=Actor
    alphaBoolBlend {blendOption=Linear}
    alphaScaleBiasClamp {bInterpResult=true, interpSpeedDecreasing=4, interpSpeedIncreasing=6}
    boundBone {boneName=dyn_choker_parent}
    componentAppliedLinearAccClamp=<6000 6000 6000>
    componentLinearAccScale=<0.25 0.25 0.25>
    componentLinearVelScale=<0.25 0.25 0.25>
    physicsBodyDefinitions [{
      sphereCollisionRadius=0
      boundBone {boneName=dyn_choker_parent}
      boxExtents=<8 8 3>
      constraintSetup {angularLimitsMax=<0 60 0>, angularLimitsMin=<0 5 0>, angularTarget=<0 0 0>}
      localJointOffset=<2 0 0>
     }]
    relativeSpaceBone {boneName=spine_05}
   }
   Output Pose(Result=Component To Local(ComponentPose=@n23)) {layerGroup=None}
  ]}
 EventGraph {nodes [
   Event BlueprintUpdateAnimation()
   > n1 = Sequence()
   n2 = GetRealTimeSeconds()
   n3 = "Sin (Radians)"(A=float * float(A=float * float(A=1, B=1.5), B=@n2))
   n4 = "Sin (Radians)"(A=float * float(A=float * float(A=1, B=1), B=@n2))
   n5 = "Sin (Radians)"(A=float * float(A=float * float(A=1, B=2), B=@n2))
   > Set Spike_Float_A(Spike_Float_A:Vector=MakeVector(X=float * float(A=@n3, B=1.75), Y=float * float(A=@n4, B=0.56), Z=float * float(A=@n5, B=0.22))) -> (Output_Get:Vector)
   n6 = GetRealTimeSeconds()
   n7 = "Sin (Radians)"(A=float * float(A=float * float(A=1, B=1), B=@n6))
   n8 = "Sin (Radians)"(A=float * float(A=float * float(A=1, B=2), B=@n6))
   n9 = "Sin (Radians)"(A=float * float(A=float * float(A=1, B=1.5), B=@n6))
   > Set Spike_Float_B(Spike_Float_B:Vector=MakeVector(X=float * float(A=@n9, B=0.75), Y=float * float(A=@n7, B=0.56), Z=float * float(A=@n8, B=0.22))) -> (Output_Get:Vector)
   Set Shoulderpads_ADS_Alpha(execute=@n1.then_1, Shoulderpads_ADS_Alpha:double=SelectFloat(A=1, B=0, bPickA=Get bIsTargeting())) -> (Output_Get:double)
  ]}
]
variables [
 Shoulderpads_ADS_Alpha {type=number}
 Spike_Float_A {type=Vector}
 Spike_Float_B {type=Vector}
]
