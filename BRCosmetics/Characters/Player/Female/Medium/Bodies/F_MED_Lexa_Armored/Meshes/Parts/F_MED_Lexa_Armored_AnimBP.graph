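# Animation Blueprint for the armored Lexa female cosmetic body that makes the head ring bone float and lag with motion.
# It copies the pose from the attached parent mesh, then runs a low-alpha, zero-gravity spring AnimDynamics on dyn_head_ring_parent, relative to neck_01.
# An additive bone-space translation then offsets the ring by Ring_Float_A, which the event graph rebuilds each update from three time-based sine waves at different frequencies.
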AnimBlueprint /BRCosmetics/Characters/Player/Female/Medium/Bodies/F_MED_Lexa_Armored/Meshes/Parts/F_MED_Lexa_Armored_AnimBP
target_skeleton=&/F_MED_Lexa_Armored_Skeleton
graphs [
 AnimGraph {nodes [
   n1 = Anim Dynamics - Bone: dyn_head_ring_parent(ComponentPose=Local To Component(LocalPose=Copy Pose From Mesh() {bUseAttachedParent=true}), Alpha=0.1, ExternalForce="0,0,0") {
    angularDampingOverride=0.9
    angularSpringConstant=150
    bAngularSpring=true
    bGravityOverrideInSimSpace=true
    bLinearSpring=true
    bOverrideAngularDamping=true
    bOverrideLinearDamping=true
    gravityScale=0
    linearDampingOverride=0.995
    linearSpringConstant=50
    simulationSpace=BoneRelative
    alphaBoolBlend {blendOption=Linear}
    boundBone {boneName=dyn_head_ring_parent}
    componentLinearAccScale=<0.1 0.1 0.3>
    componentLinearVelScale=<0.8 0.8 0.6>
    physicsBodyDefinitions [{
      sphereCollisionRadius=0
      boundBone {boneName=dyn_head_ring_parent}
      boxExtents=<50 20 50>
      constraintSetup {angularLimitsMax=<0 0 0>, angularLimitsMin=<0 0 0>, angularTarget=<0 0 0>, linearAxesMax=<2 2 2>, linearAxesMin=<-2 -2 -2>}
      localJointOffset=<0 0 0>
     }]
    relativeSpaceBone {boneName=neck_01}
   }
   n2 = "Transform (Modify) Bone - Bone: dyn_head_ring_parent"(ComponentPose=@n1, Alpha=1, Translation=Get Ring_Float_A() -> (Ring_Float_A:Vector), Rotation="0,0,0", Scale="1,1,1") {
    translationMode=BMM_Additive
    translationSpace=BCS_BoneSpace
    alphaBoolBlend {blendOption=Linear}
    boneToModify {boneName=dyn_head_ring_parent}
   }
   Output Pose(Result=Component To Local(ComponentPose=@n2)) {layerGroup=None}
  ]}
 EventGraph {nodes [
   Event BlueprintUpdateAnimation()
   n1 = GetRealTimeSeconds()
   n2 = "Sin (Radians)"(A=float * float(A=float * float(A=1, B=1.5), B=@n1))
   n3 = "Sin (Radians)"(A=float * float(A=float * float(A=1, B=1), B=@n1))
   n4 = "Sin (Radians)"(A=float * float(A=float * float(A=1, B=0.3), B=@n1))
   > Set Ring_Float_A(Ring_Float_A:Vector=MakeVector(X=float * float(A=@n2, B=0.3), Y=float * float(A=@n3, B=0.5), Z=float * float(A=@n4, B=0.35))) -> (Output_Get:Vector)
  ]}
]
variables [Ring_Float_A {type=Vector}]
